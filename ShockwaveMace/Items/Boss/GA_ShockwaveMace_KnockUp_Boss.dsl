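; Boss variant of the Shockwave Mace knock-up gameplay ability, by its name, with no logic of its own in the event graph.
; Activation and failed passive activation only forward to the parent class, which is not shown here.
; The end-ability event is empty and has no parent call.
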(event Ability|EventActivateAbility
  (|CallParentFunction))

(event Ability|EventOnEndAbility
  (params (bWasCancelled :type Boolean)))

(event Ability|EventFailedToActivatePassiveAbility
  (params (ActorInfo :type Gameplay%20Ability%20Actor%20Info%20Structure))
  (|CallParentFunction ActorInfo))
